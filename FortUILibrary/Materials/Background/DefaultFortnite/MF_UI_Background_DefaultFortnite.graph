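# Material function that produces the Color for the default Fortnite UI background.
# A radial exponential gradient on Vault World UV helper coordinates, offset off-center, blends two colors from the M_UI_SceneBackgroundMCP parameter collection.
# Named reroutes carry screen UV and two masks, a vertical fade toward top and bottom and a horizontal falloff from center, which appear to multiply the blend as a vignette.
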MaterialFunction /FortUILibrary/Materials/Background/DefaultFortnite/MF_UI_Background_DefaultFortnite
graph {nodes [
  n1 = NamedRerouteUsage()
  n2 = NamedRerouteUsage()
  n3 = MaterialFunctionCall!(Use Local UV=Constant(), Texture Size, Scale Pivot Point) {materialFunction=/Game/Environments/FrontEnd/Blueprints/ItemPreview/VaultWorldTransitions/MaterialFunctions/MF_UI_VaultWorld_UV}
  n4 = AppendVector(A=Constant() {r=0.65}, B=Constant() {r=0.35})
  n5 = LinearInterpolate(
   A=CollectionParameter() {collection=/Game/UI/Foundation/Environment/Shell/Materials/M_UI_SceneBackgroundMCP, parameterName=RGBA0}
   B=CollectionParameter() {collection=&M_UI_SceneBackgroundMCP, parameterName=RGBA1}
   Alpha=MaterialFunctionCall(UVs=Subtract(A=@n3.Scale Y, B=@n4), Radius=Constant() {r=2}, Density=Constant() {r=1.5}) {materialFunction=&RadialGradientExponential}.RadialGradientExponential
  )
  FunctionOutput(None=Multiply(A=Multiply(A=@n1, B=@n2), B=ComponentMask(None=@n5) {b=true, g=true, r=true})) {outputName=Color}
  NamedRerouteDeclaration(None=@n3.UV) {name=ScreenUV, nodeColor=<rgba 0 0 0 1>}
  n6 = NamedRerouteUsage()
  n7 = OneMinus(None=@n6)
  n8 = Multiply(A=Add(A=ComponentMask(None=@n7) {g=true}) {constB=0.125}) {constB=1.5}
  n9 = Add(A=ComponentMask(None=@n6) {g=true}) {constB=0.15}
  NamedRerouteDeclaration(None=Multiply(A=Saturate(None=@n8), B=Saturate(None=@n9))) {name=TopGradient, nodeColor=<rgba 0 0 0 1>}
  n10 = Add(A=@n6) {constB=-0.5}
  n11 = OneMinus(None=ComponentMask(None=Abs(None=@n10)) {r=true})
  NamedRerouteDeclaration(None=Saturate(None=Add(A=@n11) {constB=0.025})) {name=MidGradient, nodeColor=<rgba 0 0 0 1>}
 ]}
